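# Material function for the DriftTrail vehicle cosmetic that biases an input color, likely a product paint color, toward a more saturated tint with controlled brightness.
# It raises Color to SaturationPower and normalizes the result, separating a saturated hue direction from magnitude.
# That vector is scaled between BaseBrightness and BaseBrightness plus Range, blended per channel by the saturated, scaled-down original color, so stronger channels come out brighter.
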MaterialFunction /VehicleCosmetics/DriftTrail/DriftTrail_Binary/Materials/MF_PaintedProductColorBias
graph {nodes [
  n1 = FunctionInput(Preview=Constant3Vector() {constant=<rgba 0.8 0.25 0 0>}) {inputName=Color}
  n2 = Constant() {r=1.7}
  n3 = Normalize(VectorInput=Power(Base=@n1, Exp=FunctionInput(Preview=@n2) {inputName=SaturationPower, inputType=FunctionInput_Scalar}))
  n4 = FunctionInput(Preview=Constant()) {inputName=BaseBrightness, inputType=FunctionInput_Scalar}
  n5 = Add(A=@n4, B=FunctionInput(Preview=Constant() {r=3.2}) {inputName=Range, inputType=FunctionInput_Scalar})
  n6 = Divide(A=@n1, B=Constant3Vector() {constant=<rgba 1.25 1.25 1.25 0>})
  FunctionOutput(None=LinearInterpolate(A=Multiply(A=@n3, B=@n4), B=Multiply(A=@n3, B=@n5), Alpha=Saturate(None=@n6)))
 ]}
