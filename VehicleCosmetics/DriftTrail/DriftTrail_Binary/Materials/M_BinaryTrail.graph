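# Translucent unlit, two-sided material for the Binary drift trail vehicle cosmetic, usable on Niagara ribbons and mesh particles.
# Panning binary-tech textures combine with scrolling modulation masks, tinted by particle color through a painted-product color bias for emissive.
# UV gradients and particle alpha shape the opacity fade along edges and length.
# Two dynamic parameter sets drive hardness, blur, brightness and falloff, and a static switch enables pattern-based distortion.
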Material /VehicleCosmetics/DriftTrail/DriftTrail_Binary/Materials/M_BinaryTrail
graph {
 properties {
  bUsedWithNiagaraMeshParticles=true
  bUsedWithNiagaraRibbons=true
  blendMode=BLEND_Translucent
  pixelDepthOffsetMode=PDOM_Legacy
  shadingModel=MSM_Unlit
  twoSided=true
  usedShadingModels=Unlit
 }
 nodes [
  n1 = ParticleColor()
  n2 = MaterialFunctionCall!(Color=@n1.RGB, BaseBrightness=Constant() {r=0.1}, Range=Constant() {r=0.2}, SaturationPower=ScalarParameter() {parameterName=Saturation}) {materialFunction=&/MF_PaintedProductColorBias}
  n3 = TextureCoordinate() {uTiling=-6, vTiling=-1}
  n4 = Constant() {r=0.5}
  n5 = Multiply(A=Time(), B=ScalarParameter() {defaultValue=1, parameterName=PanSpeed})
  n6 = TextureSample(UVs=Panner(Coordinate=Panner(Coordinate=@n3, Time=@n4) {speedX=0.2, speedY=0.23}, Time=@n5) {speedX=0.05}) {texture=/DealershipSharedLibrary/FX/Textures/Diffuse/T_VCVFX_Trail_Binary_Binary_Tech_RGB}
  n7 = TextureSample(UVs=Panner(Coordinate=TextureCoordinate() {uTiling=4}, Time=@n5) {speedX=-0.05}) {texture=&T_VCVFX_Trail_Binary_Binary_Tech_RGB}
  n8 = Desaturation(None=Add(A=@n6.RGB, B=@n7.RGB))
  n9 = DynamicParameter() {paramNames [Hardness,BlurAmount,AddMultiply,Brightness]}
  n10 = TextureSample(UVs=Panner(Coordinate=TextureCoordinate() {uTiling=2}, Time=@n5) {speedX=-0.5}) {texture=/VehicleCosmetics/DriftTrail/DriftTrail_Binary/Textures/T_VCVFX_Trail_Binary_PanTime02_RGB}
  n11 = TextureSample(UVs=Panner(Coordinate=TextureCoordinate() {uTiling=1.75}) {speedX=-0.35}) {texture=&T_VCVFX_Trail_Binary_PanTime02_RGB}
  n12 = Add(
   A=Multiply(A=Multiply(A=@n9.BlurAmount, B=@n6.B), B=Desaturation(None=@n10.RGB))
   B=Multiply(A=Multiply(A=@n9.BlurAmount, B=@n7.B), B=Desaturation(None=@n11.RGB))
  )
  n13 = Add(A=@n8, B=Multiply(A=Multiply(A=@n8, B=@n9.Hardness), B=@n12))
  n14 = TextureCoordinate()
  n15 = ComponentMask(None=@n14) {g=true}
  n16 = OneMinus(None=@n15)
  n17 = Multiply(A=Saturate(None=ConstantBiasScale(None=@n15) {bias=-0.01, scale=5}), B=Saturate(None=ConstantBiasScale(None=@n16) {bias=-0.01, scale=5}))
  n18 = Add(A=@n13, B=Multiply(A=NamedRerouteUsage(), B=@n17))
  MaterialFunctionCall(Emissive=Multiply(A=Multiply(A=@n2.Result, B=@n18), B=NamedRerouteUsage())) {materialFunction=&MF_FX_EmissiveAdjustments}
  n19 = Multiply(A=Constant() {r=0.3}, B=@n12)
  n20 = Saturate(None=@n13)
  n21 = TextureSample(UVs=Panner(Coordinate=TextureCoordinate() {uTiling=-0.25}, Time=Constant() {r=0.5}) {speedX=-0.1, speedY=1}) {texture=&T_VCVFX_Trail_Binary_PanTime02_RGB}
  n22 = ComponentMask(None=@n14) {r=true}
  n23 = Multiply(A=Add(A=@n21.B, B=@n21.B), B=Saturate(None=ConstantBiasScale(None=@n22) {bias=-0.01, scale=5}))
  n24 = ConstantBiasScale(None=@n22) {bias=-0.01, scale=5}
  n25 = Multiply(A=Multiply(A=OneMinus(None=@n22), B=Saturate(None=@n24)), B=@n17)
  n26 = Add(A=Multiply(A=Add(A=@n20, B=@n19), B=Multiply(A=@n23, B=@n25)), B=Saturate(None=ConstantBiasScale(None=@n25) {bias=0, scale=1.2}))
  n27 = Power(Base=ComponentMask(None=TextureCoordinate()) {r=true}) {constExponent=0.09}
  n28 = ComponentMask(None=TextureCoordinate()) {r=true}
  n29 = DynamicParameter() {parameterIndex=1, paramNames [Power Negative, Multiply PNegative, DivideNegative, Brightness]}
  n30 = Multiply(A=Power(Base=OneMinus(None=@n28), Exp=@n29.Power Negative) {constExponent=0.09}, B=@n29.Multiply PNegative)
  Multiply(A=Saturate(None=Multiply(A=@n1.A, B=@n26)), B=Multiply(A=Multiply(A=@n27), B=Divide(A=@n30, B=@n29.DivideNegative)))
  NamedRerouteDeclaration(None=@n9.AddMultiply) {name=Dynamic-AddMultiply, nodeColor=<rgba 1 0 0.9844 1>}
  NamedRerouteDeclaration(None=@n9.Brightness) {name=Dynamic-Brightness, nodeColor=<rgba 1 0 0.9844 1>}
  StaticSwitchParameter(True=Multiply(A=@n20, B=ScalarParameter() {defaultValue=10, parameterName=DistortionAmount}), False=Constant()) {parameterName=Distortion}
 ]
}
