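# Unlit, translucent, two-sided Niagara ribbon material for the Binary drift trail vehicle cosmetic.
# Binary texture channels are masked by a PanTime texture panned at two speeds, then summed into a pattern that drives emissive and particle-alpha opacity.
# TwoSidedSign selects oppositely tiled and panned UVs per face, mirroring the pattern on back faces.
# Particle color passes through a painted-color bias function, and a dynamic parameter supplies Brightness.
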Material /VehicleCosmetics/DriftTrail/DriftTrail_Binary/Materials/M_Binary_01
graph {
 properties {
  bUsedWithNiagaraRibbons=true
  blendMode=BLEND_Translucent
  pixelDepthOffsetMode=PDOM_Legacy
  shadingModel=MSM_Unlit
  twoSided=true
  usedShadingModels=Unlit
 }
 nodes [
  n1 = ParticleColor()
  n2 = Constant() {r=1.5}
  n3 = Constant() {r=2}
  n4 = Constant() {r=0.85}
  n5 = NamedRerouteUsage()
  n6 = TextureCoordinate() {uTiling=-16}
  n7 = TextureCoordinate() {uTiling=16}
  n8 = TwoSidedSign()
  n9 = TextureSample(UVs=LinearInterpolate(A=Panner(Coordinate=@n6) {speedX=0.25}, B=Panner(Coordinate=@n7) {speedX=-0.25}, Alpha=Saturate(None=@n8))) {texture=/VehicleCosmetics/DriftTrail/DriftTrail_Binary/Textures/T_VCVFX_Trail_Binary_Binary_RGB}
  n10 = TextureSample(UVs=Panner(Coordinate=TextureCoordinate() {uTiling=2}) {speedX=-0.5}) {texture=/VehicleCosmetics/DriftTrail/DriftTrail_Binary/Textures/T_VCVFX_Trail_Binary_PanTime_RGB}
  n11 = TextureSample(UVs=Panner(Coordinate=TextureCoordinate()) {speedX=-0.75}) {texture=&T_VCVFX_Trail_Binary_PanTime_RGB}
  n12 = Add(A=Multiply(A=@n9.R, B=@n10.R), B=Add(A=Multiply(A=@n9.G, B=@n11.G), B=Multiply(A=@n9.B, B=@n10.B)))
  MaterialFunctionCall(Emissive=Multiply(
    A=Multiply(A=@n5, B=@n12)
    B=MaterialFunctionCall!(Color=@n1.RGB, BaseBrightness=@n2, Range=@n3, SaturationPower=@n4) {materialFunction=&/MF_PaintedProductColorBias}.Result
   )) {materialFunction=&MF_FX_EmissiveAdjustments}
  Multiply(A=@n1.A, B=@n12)
  n13 = DynamicParameter() {paramNames [TextureSwitch, TextureSwitch, Brightness, Param4]}
  NamedRerouteDeclaration(None=@n13.Brightness) {name=Dynamic-Brightness, nodeColor=<rgba 1 0 0.7266 1>}
  Reroute(None=@n13.TextureSwitch)
 ]
}
