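# Translucent unlit Niagara sprite material for the Binary drift trail cosmetic, drawing flickering binary-texture glyphs.
# Screen-space panning noise and a scan texture slightly distort the sprite UVs that sample the binary texture, whose R and B channels are blended by dynamic parameter TextureSwitch.
# The same noise adds bright flicker to that mask, which is scaled by Brightness and particle color through MF_FX_EmissiveAdjustments.
# Particle alpha times the mask forms a separate opacity term.
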Material /VehicleCosmetics/DriftTrail/DriftTrail_Binary/Materials/M_Binary_02
graph {
 properties {
  bUsedWithNiagaraSprites=true
  blendMode=BLEND_Translucent
  pixelDepthOffsetMode=PDOM_Legacy
  shadingModel=MSM_Unlit
  twoSided=true
  usedShadingModels=Unlit
 }
 nodes [
  n1 = DynamicParameter() {paramNames [TextureSwitch, TextureSwitch, Brightness, Param4]}
  n2 = Constant() {r=0.01}
  n3 = ComponentMask(None=ScreenPosition().ViewportUV) {g=true, r=true}
  n4 = Divide(A=@n3, B=Constant2Vector() {g=0.1, r=0.25})
  n5 = TextureSample(UVs=Panner(Coordinate=@n4) {speedY=2}) {texture=/DealershipSharedLibrary/FX/Textures/Noise/T_VCVFX_NoiseGeneric01Pack_RGB}
  n6 = TextureSample(UVs=Panner(Coordinate=@n4) {speedY=-2}) {texture=&T_VCVFX_NoiseGeneric01Pack_RGB}
  n7 = Constant2Vector() {g=0.1, r=0.5}
  n8 = TextureSample(UVs=Panner(Coordinate=Divide(A=@n3, B=@n7)) {speedY=0.05}) {texture=&T_VCVFX_NoiseGeneric01Pack_RGB}
  n9 = TextureSample(UVs=Panner(Coordinate=@n3) {speedY=-0.1}) {texture=/DealershipSharedLibrary/FX/Textures/TexturePackage/T_VCVFX_Trail_Binary_Scan_RGB}
  n10 = Add(A=Multiply(A=@n8.R, B=Add(A=@n5.G, B=@n6.G)), B=Multiply(A=@n9.G, B=@n9.G))
  n11 = TextureSample(UVs=Add(A=TextureCoordinate(), B=Multiply(A=@n2, B=@n10))) {texture=/VehicleCosmetics/DriftTrail/DriftTrail_Binary/Textures/T_VCVFX_Trail_Binary_Binary_RGB}
  n12 = LinearInterpolate(A=LinearInterpolate(A=@n11.R, B=@n11.R, Alpha=@n1.TextureSwitch), B=@n11.B, Alpha=@n1.TextureSwitch)
  n13 = Multiply(A=@n8.R, B=Add(A=Multiply(A=@n5.R, B=@n6.G), B=Multiply(A=@n5.G, B=@n6.R)))
  n14 = Constant() {r=20}
  n15 = Add(A=Multiply(A=Constant() {r=0.1}, B=@n12), B=Multiply(A=@n12, B=Multiply(A=@n13, B=@n14)))
  n16 = ParticleColor()
  MaterialFunctionCall(Emissive=Multiply(A=Multiply(A=@n1.Brightness, B=@n15), B=@n16.RGB)) {materialFunction=&MF_FX_EmissiveAdjustments}
  Multiply(A=@n16.A, B=@n12)
 ]
}
