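# Additive, two-sided master material for Niagara mesh particles, drawing the animated tire-edge glow of a vehicle drift trail cosmetic.
# Two parameterized texture layers pan over 90-degree rotated UVs, with red from the first and green from the second separately tinted and boosted.
# Their sum is multiplied by particle color and passed through MF_FX_EmissiveAdjustments.
# An alpha chain multiplies the blended channels by a powered U gradient and particle alpha.
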Material /VehicleCosmetics/DriftTrail/DriftTrail_Binary/Materials/M_Master_TireEdge
graph {
 properties {bUsedWithNiagaraMeshParticles=true, blendMode=BLEND_Additive, pixelDepthOffsetMode=PDOM_Legacy, twoSided=true, usedShadingModels=Default Lit}
 nodes [
  n1 = ParticleColor()
  n2 = VectorParameter() {group=Texture1, parameterName=Texture1_Tile, defaultValue=<rgba 1 3 0 1>}
  n3 = MaterialFunctionCall(R=@n2.R, G=@n2.R, B=@n2.R) {materialFunction=&AppendMany}
  n4 = MaterialFunctionCall(UVs=TextureCoordinate(), "Rotation Angle (0-1)"=Constant() {r=0.25}) {materialFunction=&CustomRotator}
  n5 = ScalarParameter() {defaultValue=1, group=Texture1, parameterName=Texture1_Time}
  n6 = Time()
  n7 = VectorParameter() {group=Texture1, parameterName=Texture1_Speed, defaultValue=<rgba 0.2 1.5 0 1>}
  n8 = TextureSampleParameter2D(UVs=Panner(
    Coordinate=Multiply(A=@n3.RG, B=@n4.Rotated Values)
    Time=Multiply(A=@n5, B=@n6)
    Speed=MaterialFunctionCall(R=@n7.R, G=@n7.R, B=@n7.R) {materialFunction=&AppendMany}.RG
   ) {speedX=0.2, speedY=1.5}) {
   group=Texture1
   parameterName=Texture1
   texture=/VehicleCosmetics/DriftTrail/DriftTrail_StarBeam/Textures/T_VCVFX_Trail_StarBeam_DriftWheelStars_A
  }
  n9 = Multiply(
   A=ScalarParameter() {defaultValue=15, group=Texture1, parameterName=Texture1_ColorBoost}
   B=Multiply(A=VectorParameter() {group=Texture1, parameterName=Texture1_ColorOverlay, defaultValue=<rgba 0 0.5639 1 1>}.RGB, B=@n8.R)
  )
  n10 = VectorParameter() {group=Texture2, parameterName=Texture2_Tile, defaultValue=<rgba 1 3 0 1>}
  n11 = MaterialFunctionCall(R=@n10.R, G=@n10.R, B=@n10.R) {materialFunction=&AppendMany}
  n12 = ScalarParameter() {defaultValue=1, group=Texture2, parameterName=Texture2_Time}
  n13 = Time()
  n14 = VectorParameter() {group=Texture2, parameterName=Texture2_Speed, defaultValue=<rgba 0.3 1.6 0 1>}
  n15 = TextureSampleParameter2D(UVs=Panner(
    Coordinate=Multiply(A=@n4.Rotated Values, B=@n11.RG)
    Time=Multiply(A=@n12, B=@n13)
    Speed=MaterialFunctionCall(R=@n14.R, G=@n14.R, B=@n14.R) {materialFunction=&AppendMany}.RG
   ) {speedX=1, speedY=1.4}) {
   group=Texture2
   parameterName=Texture2
   texture=&T_VCVFX_Trail_StarBeam_DriftWheelStars_A
  }
  n16 = Multiply(
   A=ScalarParameter() {defaultValue=15, group=Texture2, parameterName=Texture2_ColorBoost}
   B=Multiply(A=VectorParameter() {group=Texture2, parameterName=Texture2_ColorOverlay, defaultValue=<rgba 1 0.949 0 1>}.RGB, B=@n15.G)
  )
  MaterialFunctionCall(Emissive=Multiply(A=@n1.RGB, B=Add(A=@n9, B=@n16))) {materialFunction=&MF_FX_EmissiveAdjustments}
  n17 = LinearInterpolate(A=@n8.R, B=@n15.G, Alpha=ScalarParameter() {defaultValue=0.5, parameterName=TextureADD_Blend})
  n18 = ScalarParameter() {defaultValue=5, parameterName=Combined_TextureAmount}
  n19 = ComponentMask(None=TextureCoordinate()) {r=true}
  n20 = ScalarParameter() {defaultValue=5, parameterName=Opactity_GradientPower}
  Multiply(A=@n1.A, B=Multiply(A=Multiply(A=@n17, B=@n18), B=Power(Base=@n19, Exp=@n20)))
 ]
}
