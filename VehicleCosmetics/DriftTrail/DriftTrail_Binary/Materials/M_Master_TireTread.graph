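# Master unlit translucent material for tire tread drift trail effects on Niagara ribbons and mesh particles.
# Two panning textures with per-layer tiling, speed, tint and boost are added, multiplied by particle color and passed through an emissive adjustment function.
# Alpha blends the same textures, applies power and boost, then multiplies by a static tread mask and particle alpha.
# The RG static switch feeds the same red channel to both branches.
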Material /VehicleCosmetics/DriftTrail/DriftTrail_Binary/Materials/M_Master_TireTread
graph {
 properties {
  bUsedWithNiagaraMeshParticles=true
  bUsedWithNiagaraRibbons=true
  blendMode=BLEND_Translucent
  pixelDepthOffsetMode=PDOM_Legacy
  shadingModel=MSM_Unlit
  usedShadingModels=Unlit
 }
 nodes [
  n1 = ParticleColor()
  n2 = VectorParameter() {group=Texture1, parameterName=Texture1_Tile, defaultValue=<rgba 1 1 0 1>}
  n3 = MaterialFunctionCall(R=@n2.R, G=@n2.R, B=@n2.R) {materialFunction=&AppendMany}
  n4 = TextureCoordinate()
  n5 = ScalarParameter() {defaultValue=1, group=Texture1, parameterName=Texture1_Time}
  n6 = Time()
  n7 = VectorParameter() {group=Texture1, parameterName=Texture1_Speed, defaultValue=<rgba 0.5 0 0 1>}
  n8 = TextureSampleParameter2D(UVs=Panner(
    Coordinate=Multiply(A=@n3.RG, B=@n4)
    Time=Multiply(A=@n5, B=@n6)
    Speed=MaterialFunctionCall(R=@n7.R, G=@n7.R, B=@n7.R) {materialFunction=&AppendMany}.RG
   ) {speedX=0.2, speedY=1.5}) {group=Texture1, parameterName=Texture1, texture=/DealershipSharedLibrary/FX/Textures/Noise/T_VCVFX_NoiseFire02_RGB}
  n9 = Multiply(
   A=ScalarParameter() {defaultValue=8, group=Texture1, parameterName=Texture1_ColorBoost}
   B=Multiply(
    A=VectorParameter() {group=Texture1, parameterName=Texture1_ColorOverlay, defaultValue=<rgba 1 1 1 1>}.RGB
    B=StaticSwitchParameter(True=@n8.R, False=@n8.R) {parameterName=RG}
   )
  )
  n10 = VectorParameter() {group=Texture2, parameterName=Texture2_Tile, defaultValue=<rgba 1 3 0 1>}
  n11 = MaterialFunctionCall(R=@n10.R, G=@n10.R, B=@n10.R) {materialFunction=&AppendMany}
  n12 = ScalarParameter() {defaultValue=1, group=Texture2, parameterName=Texture2_Time}
  n13 = Time()
  n14 = VectorParameter() {group=Texture2, parameterName=Texture2_Speed, defaultValue=<rgba 2 0.5 0 1>}
  n15 = TextureSampleParameter2D(UVs=Panner(
    Coordinate=Multiply(A=@n4, B=@n11.RG)
    Time=Multiply(A=@n12, B=@n13)
    Speed=MaterialFunctionCall(R=@n14.R, G=@n14.R, B=@n14.R) {materialFunction=&AppendMany}.RG
   ) {speedX=1, speedY=1.4}) {
   group=Texture2
   parameterName=Texture2
   texture=/VehicleCosmetics/Athena/Effects/Particles/Skydiving/NS_Hightower_Date/T_Hightower_Ribbon_Color_Vertical
  }
  n16 = Multiply(
   A=ScalarParameter() {defaultValue=5, group=Texture2, parameterName=Texture2_ColorBoost}
   B=Multiply(A=VectorParameter() {group=Texture2, parameterName=Texture2_ColorOverlay, defaultValue=<rgba 0.6667 0.6667 0.6667 1>}.RGB, B=@n15.G)
  )
  MaterialFunctionCall(Emissive=Multiply(A=@n1.RGB, B=Add(A=@n9, B=@n16))) {materialFunction=&MF_FX_EmissiveAdjustments}
  n17 = ScalarParameter() {defaultValue=15, parameterName=Alpha_Boost}
  n18 = Power(
   Base=LinearInterpolate(A=@n8.R, B=@n15.G, Alpha=ScalarParameter() {defaultValue=0.5, parameterName=TextureADD_Blend})
   Exp=ScalarParameter() {defaultValue=2, parameterName=Opactity_GradientPower}
  )
  Multiply(
   A=@n1.A
   B=Multiply(
    A=Multiply(A=@n17, B=@n18) {constA=15}
    B=TextureSampleParameter2D() {parameterName=TireTread_Mask, texture=/VehicleCosmetics/DriftTrail/DriftTrail_Wheel/Textures/T_Tire_Tread_Mask}.R
   )
  )
 ]
}
